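# Conversation database for the Stage Slayer NPC, defining the greeting a player gets when talking to this character.
# The entry tag leads to one speech node with a default line and contextual lines prioritized by curve rows.
# Contextual lines react to time since the last match, a first-ever conversation, or wearing specific outfits.
# The speech node then links out by tag to the shared NPC Library quests and basic services conversations.
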ConversationDatabase /NPCLibrary/NPCs/StageSlayer/NPCLibrary_Conversation_StageSlayer
entry_points [{tag=Conversation.Entry.StageSlayer, destinations [#1]}]
nodes [
 ConversationEntryPointNode_7 {
  id=#1
  class=ConversationEntryPointNode
  outputs [#2]
  properties {entryTag {tagName=Conversation.Entry.StageSlayer}}
 }
 Speech_C_4 {
  id=#2
  class=Speech_C
  outputs [#3, #4]
  properties {
   nodeName=""
   generalConfig {
    defaultMessage=You say something?
    contextualMessages [
     {
      message="Dude! I was wondering when you'd <b>back for your next encore!</> Righteous."
      contextRequirements [{requirement {class=FortControllerRequirement_CheckTimeSinceLastPlayedMatch}}]
      priority {curve {rowName=Tandem.Speech.Priority.WhileSincePlayed}}
     }
     {
      message="Hey, a new act! What's up?! *air guitars furiously*"
      contextRequirements [{requirement {class=FortControllerRequirement_IsFirstEverConversationWithNPC}}]
      priority {curve {rowName=Tandem.Speech.Priority.FirstEverInteraction}}
     }
     {
      message="Dude. My man! <b>My. MAN.</>"
      contextRequirements [{requirement {
         class=FortControllerRequirement_HasCID
         properties {allowedCIds [/BRCosmetics/Athena/Items/Cosmetics/Characters/CID_233_Athena_Commando_M_FortniteDJ]}
        }}]
      priority {curve {rowName=Tandem.Speech.Priority.OutfitReaction}}
     }
     {
      message="Respect. "
      contextRequirements [{requirement {
         class=FortControllerRequirement_HasCID
         properties {allowedCIds [/BRCosmetics/Athena/Items/Cosmetics/Characters/CID_097_Athena_Commando_F_RockerPunk]}
        }}]
      priority {curve {rowName=Tandem.Speech.Priority.OutfitReaction}}
     }
    ]
   }
  }
 }
 ConversationLinkNode_0 {id=#4, class=ConversationLinkNode, properties {remoteEntryTag {tagName=Conversation.Entry.BasicServices}}}
 ConversationLinkNode_1 {id=#3, class=ConversationLinkNode, properties {remoteEntryTag {tagName=Conversation.Entry.NPCLibrary.Quests}}}
]
